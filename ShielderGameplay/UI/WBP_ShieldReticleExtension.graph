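# Reticle extension widget that shows the equipped weapon's ammo as a radial progress image offset beside the reticle, presumably for the Shielder shield weapon.
# Construct caches the weapon, binds to its local ammo change event and refreshes; Destruct unbinds if that weapon is valid.
# Ammo change and clip attribute events also call UpdateAmmoCount, which sets the dynamic material's Percent to ammo divided by bullets per clip.
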WidgetBlueprint /ShielderGameplay/UI/WBP_ShieldReticleExtension
graphs [
 EventGraph {nodes [
   n1 = OnLocalAmmoChanged!() -> (OutputDelegate, then, LocalCount:int, LocalRemaining:int) {is=title K2Node_CustomEvent}
   > UpdateAmmoCount!(
    execute
    self
    RemainingAmmo:int=@n1.LocalCount
    BulletsPerClip:int=GetBulletsPerClip!(self:Fort Weapon Object Reference=GetEquippedWeapon!(self:Fort Weapon Reticle Extension Widget Base Object Reference) -> (ReturnValue:Fort Weapon Object Reference)) -> (ReturnValue:int=0)
   ) -> (then)
   Event Destruct()
   n2 = Get EquippedWeapon() -> (EquippedWeapon:Fort Weapon Object Reference)
   > Is Valid(InputObject=@n2)
   n3 = Create Event()
   > Unbind Event from On Local Ammo Changed(self:Fort Weapon Object Reference=@n2, Delegate=@n3) {is=title K2Node_RemoveDelegate}
   Event Construct()
   > n4 = Set EquippedWeapon(EquippedWeapon:Fort Weapon Object Reference=GetEquippedWeapon()) -> (Output_Get:Fort Weapon Object Reference)
   > Bind Event to On Local Ammo Changed(self:Fort Weapon Object Reference=@n4, Delegate=@n3) {is=title K2Node_AddDelegate}
   n5 = GetEquippedWeapon()
   > UpdateAmmoCount(
    RemainingAmmo=GetLocalRemainingAmmo!(self:Fort Weapon Object Reference=@n5) -> (ReturnValue:int=0)
    BulletsPerClip=GetBulletsPerClip(self=@n5) -> (ReturnValue=0)
   )
   Event OnAmmoClipAttributeChanged!() -> (OutputDelegate, then, AmmoClipValue:float) {is=title K2Node_Event}
   n6 = GetEquippedWeapon()
   > UpdateAmmoCount(RemainingAmmo=GetLocalRemainingAmmo(self=@n6) -> (ReturnValue=0), BulletsPerClip=GetBulletsPerClip(self=@n6) -> (ReturnValue=0))
  ]}
 UpdateAmmoCount {nodes [
   n1 = UpdateAmmoCount!() -> (then, RemainingAmmo:int, BulletsPerClip:int) {is=title K2Node_FunctionEntry}
   > n2 = GetDynamicMaterial(self=Get Image_Progress() -> (Image_Progress:Image Object Reference))
   > SetScalarParameterValue(self=@n2, ParameterName=Percent, Value=float / float(A="To Float (Integer)"(InInt=@n1), B="To Float (Integer)"(InInt=@n1.BulletsPerClip)))
  ]}
]
variables [EquippedWeapon {type=FortWeapon, default=null}]
widgets {
 info {
  inheritedWidgetCount=1
  parentClass=/Script/FortniteUI.FortWeaponReticleExtensionWidgetBase
  rootWidgetClass=&Overlay
  widgetCount=4
 }
 widgets [
  Overlay_0 {widgetClassPath=&Overlay, widget=Overlay_0}
  CommonBorder {
   widgetClassPath=&CommonBorder
   bIsVariable=true
   parent=Overlay_0
   widget=CommonBorder
   properties {padding=<ltrb 0 0 0 0>, renderTransform {angle=0, scale=<1 1>, shear=<0 0>, translation=<75 0>}}
   slot {class=OverlaySlot, properties {horizontalAlignment=HAlign_Center, verticalAlignment=VAlign_Center}}
  }
  Image_Progress {
   widgetClassPath=&Image
   bIsVariable=true
   parent=CommonBorder
   widget=Image_Progress
   properties {
    visibility=HitTestInvisible
    brush {
     bIsDynamicallyLoaded=false
     drawAs=Image
     imageType=NoImage
     mirroring=NoMirror
     resourceName=None
     resourceObject=/Game/Athena/HUD/Materials/M_UI_RadialProgress_Ammo
     tiling=NoTile
     imageSize=<64 64>
     margin=<ltrb 0 0 0 0>
     outlineSettings {
      bUseBrushTransparency=false
      roundingType=HalfHeightRadius
      width=0
      color {colorUseRule=UseColor_Specified, specifiedColor=<rgba 0 0 0 0>}
      cornerRadii=<0 0 0 0>
     }
     tintColor {colorUseRule=UseColor_Specified, specifiedColor=<rgba 1 1 1 1>}
     uVRegion {bIsValid=false, max=<0 0>, min=<0 0>}
    }
    colorAndOpacity=<rgba 1 1 1 0.8>
    renderTransform {angle=135, scale=<1 1>, shear=<0 0>, translation=<0 0>}
   }
   slot {class=BorderSlot, properties {padding=<ltrb 0 0 0 0>}}
  }
  TutorialHighlight {widgetClassPath=&TutorialHighlight}
 ]
}
